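# Niagara footstep effect for snow that puffs a small burst of snow sprites at each step.
# SnowFeet launches short-lived particles under gravity, with curve-driven SubUV animation, fading colour tinted by a fake-light user parameter, and sprite size switched by an owner-speed threshold.
# Its spawn-group and LeftFoot mesh scaling remains, though only a sprite renderer is present.
# A disabled Trail emitter holds a skeletal-mesh ribbon trail with time-of-day colour and opacity.
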NiagaraSystem /GrillMoon/Assets/FX/NS_GrillMoon_MM_FootStep_Snow
dependencies {
 usedDataInterfaces [
  &NiagaraDataInterfaceCurve
  &NiagaraDataInterfaceVectorCurve
  &NiagaraDataInterfaceSpriteRendererInfo
  &NiagaraDataInterfaceColorCurve
  &NiagaraDataInterfaceSkeletalMesh
 ]
 usedDynamicInputs [&MergeHelper_VectorSwitch, &Vector2DFromFloat]
 usedModules [
  &self:SystemSpawnScript.SystemScriptSource.SystemScriptGraph.NiagaraNodeAssignment_0.SetVariables_#1
  &SystemState
 ]
}
emitters {
 SnowFeet {
  emitterUpdateScript [
   &EmitterState()
   &SpawnRate(SpawnGroup=1, SpawnRate=3, Use Spawn Probability=false) {enabled=false}
   &SpawnBurst_Instantaneous(
    Spawn Count=&RandomRangeInt(Minimum=2, Maximum=3)
    Use Spawn Probability=false
    UseLoopCountLimit=false
   )
  ]
  particleSpawnScript [
   &InitializeParticle(
    Color=<rgba 0 0 0 0.2>
    Lifetime=&RandomRangeFloat(Minimum=1.5, Maximum=1.75)
    Mesh Scale=&MaskVectorBySpawnGroup(
     Masked Vector=<1.1 1.1 1.1>
     Passthrough Vector=&RandomRangeVector(Minimum=<0.02 0.02 0.02>, Maximum=<0.05 0.05 0.05>)
    )
    Mesh Scale Mode=::NewEnumerator4
    Position Offset=<0 0 7.5>
    Position Offset Coordinate Space=::World
    Sprite Rotation Angle=&RandomRangeFloat(Minimum=0, Maximum=1)
    Sprite Rotation Mode=::NewEnumerator1
    Sprite Size Mode=::NewEnumerator1
    Uniform Sprite Size Max=45
    Uniform Sprite Size Min=35
    UsePositionOffset=true
    Write Material Random=true
   )
   /SetVariables(Particles.VisibilityTag=&SpawnGroup()) {isSetParameters=true}
   &CameraOffset(Camera Offset Amount=&RandomRangeFloat(Minimum=17.5, Maximum=20))
   &AddVelocity(
    Rotation Coordinate Space=::World
    Velocity=&RandomRangeVector(Minimum=<-25 -25 0>, Maximum=<25 25 15>)
   )
   &DynamicMaterialParameters(
    Index 0 Param 3=1
    Index 0 Param 4=1
    Module.Index0Param0Write=true
    Module.Index0Param1Write=true
    Module.Index0Param2Write=true
    Module.Index0Param3Write=true
   )
  ]
  particleUpdateScript [
   &ParticleState()
   &GravityForce(Gravity=<0 0 -90>)
   &ScaleMeshSize(Scale Factor=&VectorFromFloat(Value=&FloatFromCurve(FloatCurve=DataInterface() {propertyValues {Curve {keys [
          {arriveTangent=8.1506, interpMode=RCIM_Cubic, leaveTangent=8.1506, tangentMode=RCTM_User, time=0, value=0}
          <at 0.12 1>
          {arriveTangent=-1.1364, interpMode=RCIM_Cubic, leaveTangent=-1.1364, time=0.8789, value=1}
          <at 1 0>
         ]}}})))
   &ScaleMeshSize(Scale Factor=&MaskVectorBySpawnGroup(
     Masked Vector=&VectorFromCurve(
      VectorCurve=DataInterface() {propertyValues {XCurve {keys [<at 0 1>,<at 1 1>]}, YCurve {keys [<at 0 1>,<at 1 1>]}, ZCurve {keys [<at 0 1>,<at 1 1.1 RCIM_Cubic>]}}}
      Scale Curve=&Lerp_Vector(A=<1.4 1.4 2>, B=<2.3 1.5 2>, Alpha=$User.LeftFoot)
     )
     Passthrough Vector=<1 1 1>
    )) {module=ScaleMeshSize001}
   &SubUVAnimation(
    "Sprite Renderer "=DataInterface() {propertyValues {SpriteRenderer=&self:SnowFeet_2.NiagaraSpriteRendererProperties_1}}
    SubUV Lookup Index:NiagaraFloat=&FloatFromCurve(FloatCurve=DataInterface() {propertyValues {Curve {keys [
         {arriveTangent=4.1785, interpMode=RCIM_Cubic, leaveTangent=4.1785, tangentMode=RCTM_User, time=0, value=0.15}
         {arriveTangent=0.85, interpMode=RCIM_Cubic, leaveTangent=0.85, time=0.2, value=0.75}
         {arriveTangent=0.0224, interpMode=RCIM_Cubic, leaveTangent=0.0224, tangentMode=RCTM_User, time=1, value=1}
        ]}}})
    UseEndFrame=false
    UseStartFrame=false
   ) {module=SubUVAnimation001}
   &ScaleSpriteSize(
    Uniform Curve Scale=&Multiply_Float(
     A=2
     B=&MakeCustomFloatFromBool(
      Bool=&SetBoolByFloatComparison(A=&VectorLength(Vector=$Engine.Owner.Velocity), B=150)
      True Float=0
      False Float=1
     )
    )
    Uniform Curve Sprite Scale=DataInterface() {propertyValues {Curve {keys [
        {arriveTangent=2.6419, interpMode=RCIM_Cubic, leaveTangent=2.6419, tangentMode=RCTM_User, time=0, value=0.3}
        {interpMode=RCIM_Cubic, leaveTangent=-0.6, time=1, value=1}
       ]}}}
   )
   /SetVariables(Particles.SnowSpriteColor:LinearColor=&MultiplyLinearColor(
     Linear Color=&ColorFromCurve(ColorCurve=DataInterface() {propertyValues {
        AlphaCurve {keys [<at 0.01 0>,<at 0.1 0.3 RCIM_Cubic>,<at 1 0 RCIM_Cubic>]}
        BlueCurve {keys [<at 0 1 RCIM_Cubic>]}
        GreenCurve {keys [<at 0 1 RCIM_Cubic>]}
        RedCurve {keys [<at 0 1 RCIM_Cubic>]}
       }})
     Scale Factor=&MakeVector4_FromLinearColor(LinearColor=$User.Tint_for_FakeLight)
    )) {isSetParameters=true}
   &SolveForcesAndVelocity(Clamp Velocity=false, Limit Acceleration=false)
  ]
  rendererClasses [&NiagaraSpriteRendererProperties]
  renderers [{rendererClass=&NiagaraSpriteRendererProperties}]
 }
 Trail {
  bEnabled=false
  emitterUpdateScript [
   &EmitterState()
   &FN_SpawnAnimTrail(Spawn Mode:ENiagaraSkeletalSpawning=::NewEnumerator0)
   /SetVariables(
    Emitter.RandFloat1:NiagaraFloat=&RandomRangeFloat(Minimum=0, Maximum=1)
    Emitter.RandFloat2:NiagaraFloat=&RandomRangeFloat(Minimum=0, Maximum=1)
   ) {isSetParameters=true}
  ]
  particleSpawnScript [
   &FN_InitializeAnimTrailBoneIndecies(First Bone or Socket Index=1) {module=FN_InitializeAnimTrailBoneIndecies001}
   &FN_InitializeAnimTrail(
    AnimTrailWidth=14
    Color=<rgba 1 1 1 0.7>
    Lifetime=1
    Position Offset=&MakeVector(X=0, Y=0, Z=&Subtract_Float(A=-2.3, B=$User.TrailOffset))
    SkeletalMesh=$System.Skeletal Mesh
   ) {module=FN_InitializeAnimTrail001}
   &DynamicMaterialParameters(
    Index 0 Param 1=$Emitter.RandFloat1
    Index 0 Param 2=$Emitter.RandFloat2
    Index 0 Param 3=1.1
    Index 0 Param 4=&Multiply_Float(
     A=&FN_FrontEndFloatModifier(
      Gameplay=&FNTimeOfDayScalar(Night Scale=0.6, Morning Scale=0.7, Day Scale=0.8, Evening Scale=0.7)
      Front End=0.5
     )
     B=&MakeCustomFloatFromBool(
      Bool=&SetBoolByFloatComparison(A=&VectorLength(Vector=$Engine.Owner.Velocity), B=150)
      True Float=0
      False Float=1
     )
    )
    Module.Index0Param0Write=true
    Module.Index0Param1Write=true
    Module.Index0Param2Write=true
    Module.Index0Param3Write=true
   )
  ]
  particleUpdateScript [
   &FN_UpdateAnimTrail(SkeletalMesh=$System.Skeletal Mesh)
   &ScaleRibbonWidth(Ribbon Width Scale=&FloatFromCurve(FloatCurve=DataInterface() {propertyValues {Curve {keys [
         {interpMode=RCIM_Cubic, tangentMode=RCTM_User, time=0, value=1}
         {interpMode=RCIM_Cubic, tangentMode=RCTM_User, time=0.5, value=1}
         {arriveTangent=-7.0572, interpMode=RCIM_Cubic, leaveTangent=-7.0572, tangentMode=RCTM_User, time=1, value=0}
        ]}}})) {module=ScaleRibbonWidth002}
   &ScaleColor(
    Scale Alpha=&FloatFromCurve(FloatCurve=DataInterface() {propertyValues {Curve {keys [{interpMode=RCIM_Cubic, tangentMode=RCTM_User, time=0.1, value=0.2237}, {arriveTangent=-2.3558, tangentMode=RCTM_User, time=1, value=-1}]}}})
    Scale RGB=&MakeVectorFromLinearColorRGB(Color=&FNTimeOfDayColor(
      Night Color=<rgba 0.1954 0.2978 0.6823 1>
      Morning Color=<rgba 0.7 0.7 0.7 1>
      Day Color=<rgba 0.8 0.8 0.8 1>
      Evening Color=<rgba 0.65 0.65 0.65 1>
     ))
    ScaleA=true
    ScaleRGB=true
   ) {module=ScaleColor002}
  ]
  rendererClasses [&NiagaraRibbonRendererProperties]
  renderers [{rendererClass=&NiagaraRibbonRendererProperties}]
 }
}
summary {
 systemName=NS_GrillMoon_MM_FootStep_Snow
 emitters [
  {emitterName=SnowFeet, rendererClasses [&NiagaraSpriteRendererProperties]}
  {bEnabled=false, emitterName=Trail, rendererClasses [&NiagaraRibbonRendererProperties]}
 ]
 userVariables [
  User.bIsFrontEnd {type=NiagaraBool}
  User.bIsFrontEndPreview {type=NiagaraBool}
  User.LeftFoot {type=NiagaraFloat, defaultValue=0, description=""}
  User.RibbonIDOffset {type=NiagaraInt32}
  User.Tint_for_FakeLight {type=LinearColor, description="", defaultValue {struct=LinearColor, value=<rgba 1 1 1 1>}}
  User.TrailOffset {type=NiagaraFloat, defaultValue=0, description=""}
 ]
}
